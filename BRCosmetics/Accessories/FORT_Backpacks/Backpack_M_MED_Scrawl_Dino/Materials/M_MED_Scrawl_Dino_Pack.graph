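# Two-sided character material for the Scrawl Dino backpack, giving it an animated scribble look.
# Time-driven 2x2 flipbooks supply base color plus roughness and specular, with a flat normal, inside the shared character FX chain.
# World position offset jitters vertices by hashing time and gridded pre-skinned position, with density and strength blended by camera distance.
# A sphere-masked Y push branch is present but unconnected.
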Material /BRCosmetics/Accessories/FORT_Backpacks/Backpack_M_MED_Scrawl_Dino/Materials/M_MED_Scrawl_Dino_Pack
graph {
 properties {
  bUseMaterialAttributes=true
  bUsedWithMeshDeformer=true
  bUsedWithSkeletalMesh=true
  pixelDepthOffsetMode=PDOM_Legacy
  refractionMethod=RM_IndexOfRefraction
  twoSided=true
  usedShadingModels=Default Lit
 }
 nodes [
  Constant()
  n1 = MaterialFunctionCall() {materialFunction=&BaseFN_CharacterFunction_PreFX}
  n2 = Frac(None=Multiply(A=Time(), B=ScalarParameter() {defaultValue=2.4, parameterName=FlipbookSpeed}))
  n3 = Constant() {r=2}
  n4 = Constant() {r=2}
  n5 = TextureObject() {samplerType=SAMPLERTYPE_LinearColor, texture=/BRCosmetics/Accessories/FORT_Backpacks/Backpack_M_MED_Scrawl_Dino/Textures/M_MED_Scrawl_Dino_S}
  n6 = MaterialFunctionCall(Float3=MaterialFunctionCall("Animation  Phase (0-1)"=@n2, Number of Rows=@n3, Number of Columns=@n4, Texture=@n5) {materialFunction=&FlipBook}.Result) {materialFunction=&BreakOutFloat3Components}
  n7 = SetMaterialAttributes!(
   MaterialAttributes=MaterialFunctionCall(Gmap=@n1.Gmap, MaterialAttributes=@n1.Gmap, AO=@n1.Gmap) {materialFunction=&CharacterSkinFX}.Result
   World Position Offset=NamedRerouteUsage()
   Base Color=MaterialFunctionCall(
    "Animation  Phase (0-1)"=@n2
    Number of Rows=@n3
    Number of Columns=@n4
    Texture=TextureObject() {texture=/BRCosmetics/Accessories/FORT_Backpacks/Backpack_M_MED_Scrawl_Dino/Textures/M_MED_Scrawl_Dino_Flipbook}
   ) {materialFunction=&FlipBook}.Result
   Normal=Constant3Vector() {constant=<rgba 0 0 1 0>}
   Metallic=Constant()
   Roughness=@n6.B
   Specular=@n6.B
  ) {attributeSetTypes [
    #1
    #2
    #3
    #4
    #5
    #6
   ]}
  n8 = MaterialFunctionCall(AO=@n1.AO, Distance Blend=@n1.AO, Fresnel=@n1.AO, Gmap=@n1.AO, Pre Skinned Local Position=@n1.AO, Material=@n7) {materialFunction=&BaseFN_CharacterFunction_PostFX}
  MaterialFunctionCall(In=@n8.Result) {materialFunction=&CINE_MF_CharPostFX}
  n9 = PreSkinnedPosition()
  n10 = VertexInterpolator(VS=@n9)
  n11 = PreSkinnedLocalBounds()
  n12 = ComponentMask(None=Divide(A=ShaderStageSwitch(PixelShader=@n10.PS, VertexShader=@n9), B=Subtract(A=@n11.Max, B=@n11.Max))) {b=true, r=true}
  n13 = MaterialFunctionCall(Float2=@n12) {materialFunction=&BreakOutFloat2Components}
  n14 = Constant() {r=-0.05}
  n15 = Constant() {r=0.45}
  n16 = Constant()
  n17 = Constant() {r=-0.05}
  n18 = Constant() {r=0.55}
  n19 = Constant()
  Multiply(
   A=Multiply(A=SphereMask(A=@n13.R, B=@n14, Radius=@n15, Hardness=@n16), B=SphereMask(A=@n13.G, B=@n17, Radius=@n18, Hardness=@n19))
   B=ScalarParameter() {defaultValue=50, parameterName=Y_PushAmount}
  )
  n20 = Transform(None=Constant3Vector() {constant=<rgba 1 0 1 0>}) {transformSourceType=TRANSFORMSOURCE_World, transformType=TRANSFORM_Local}
  n21 = Multiply(
   A=Subtract(A=Time(), B=ScalarParameter() {group=OriginPrision, parameterName=Wave_X_random_seed})
   B=ScalarParameter() {defaultValue=2.4, group=OriginPrision, parameterName=Wave_X_Speed}
  ) {constB=0.6}
  n22 = Distance(A=ObjectPositionWS(), B=CameraPositionWS())
  n23 = ScalarParameter() {defaultValue=100, parameterName=FarDistance}
  n24 = ScalarParameter() {defaultValue=500, parameterName=NearDistance}
  n25 = Saturate(None=Divide(A=Subtract(A=@n22, B=@n23), B=Subtract(A=@n24, B=@n23)))
  n26 = Multiply(
   A=@n12
   B=LinearInterpolate(
    A=ScalarParameter() {defaultValue=50, parameterName=ScribleDensity_Near}
    B=ScalarParameter() {defaultValue=10, parameterName=ScribleDensity_Far}
    Alpha=@n25
   )
  ) {constB=0.1}
  n27 = MaterialFunctionCall(Float2=MaterialFunctionCall(Seed=Frac(None=@n21), Vec2=Ceil(None=@n26)) {materialFunction=&MF_Hash}.Vec2) {materialFunction=&BreakOutFloat2Components}
  n28 = Multiply(A=ComponentMask(None=@n20) {r=true}, B=@n27.R)
  n29 = Constant()
  n30 = Multiply(A=ComponentMask(None=@n20) {b=true}, B=@n27.G)
  n31 = ScalarParameter() {defaultValue=0.06, parameterName=FinalStrength_Near}
  n32 = ScalarParameter() {defaultValue=0.25, parameterName=FinalStrength_Far}
  NamedRerouteDeclaration(None=Multiply(
    A=MaterialFunctionCall(X=@n28, Y=@n29, Z=@n30) {materialFunction=&MakeFloat3}.Result
    B=LinearInterpolate(A=@n31, B=@n32, Alpha=@n25)
   )) {name=WPO, nodeColor=<rgba 0.2422 1 0 1>}
  WorldPosition()
 ]
}
